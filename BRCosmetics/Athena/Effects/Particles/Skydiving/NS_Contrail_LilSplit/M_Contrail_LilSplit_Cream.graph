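# Lit mesh-particle material for the cream part of the LilSplit ice cream skydiving contrail.
# Scrolling cell noise plus vertex blue mixes cream and brown, while vertex red switches areas to a blue-to-pink gradient.
# Time-driven, per-particle randomized sine waves build main and detail wobble and normal bulges, masked along a particle-relative Z gradient.
# Niagara dynamic parameters tune the wobbles and a directional trailing offset weighted by cubed vertex green.
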Material /BRCosmetics/Athena/Effects/Particles/Skydiving/NS_Contrail_LilSplit/M_Contrail_LilSplit_Cream
graph {
 properties {
  bUsedWithNiagaraMeshParticles=true
  pixelDepthOffsetMode=PDOM_Legacy
  previewMesh=&/SM_Contrail_LilSplit_IceCream
  usedShadingModels=Default Lit
 }
 nodes [
  n1 = Divide(A=NamedRerouteUsage()) {constB=100}
  n2 = OneMinus(None=@n1)
  n3 = Saturate(None=Multiply(A=Subtract(A=@n1) {constB=0.1}, B=Add(A=@n2) {constB=0.65}))
  n4 = Time()
  n5 = ParticleRandom()
  n6 = Add(A=Divide(A=NamedRerouteUsage()) {constB=150}, B=Multiply(A=Add(A=@n4, B=@n5), B=Constant() {r=-0.6}))
  n7 = Constant() {r=0.6}
  n8 = Divide(A=Subtract(A=Sine(None=@n6), B=@n7) {constB=0.65}, B=@n7) {constB=0.65}
  n9 = Add(A=@n2) {constB=0.2}
  n10 = Saturate(None=Multiply(A=Saturate(None=@n8), B=@n9))
  n11 = LinearInterpolate(A=NamedRerouteUsage(), B=NamedRerouteUsage(), Alpha=@n10)
  n12 = VertexNormalWS()
  n13 = DynamicParameter() {parameterIndex=2, defaultValue=<rgba 45 100 1 1>, paramNames [DetailAmplitude, DetailPeriod, DetailSpeed, WobbleOffsetRandom]}
  n14 = VertexColor()
  n15 = OneMinus(None=Ceil(None=Add(A=@n14.R, B=@n14.R)))
  n16 = LinearInterpolate(B=Multiply(A=Multiply(A=@n12, B=@n13.WobbleOffsetRandom) {constB=5}, B=Multiply(A=@n15, B=@n9)), Alpha=@n10)
  Add(A=Multiply(A=Saturate(None=@n3), B=Add(A=@n11, B=@n16)), B=NamedRerouteUsage())
  Constant()
  Constant() {r=1}
  Constant() {r=0.5}
  n17 = Time()
  n18 = ParticleRandom()
  n19 = Add(A=TextureCoordinate(), B=Multiply(A=Add(A=@n17, B=@n18), B=Constant2Vector() {g=1, r=0.25}))
  n20 = Add(
   A=VertexColor().B
   B=Power(Base=TextureSample(UVs=@n19) {samplerType=SAMPLERTYPE_Grayscale, texture=/BRCosmetics/Weapons/FORT_Melee/Pickaxe_Cube_Ninja_Male/FX/T_CellNoise}.R)
  )
  n21 = VertexColor()
  LinearInterpolate(
   A=LinearInterpolate(
    A=Constant3Vector() {constant=<rgba 1 0.9882 0.9688 1>}
    B=Constant3Vector() {constant=<rgba 0.2292 0.067 0.0465 1>}
    Alpha=Saturate(None=@n20)
   )
   B=LinearInterpolate(A=Constant3Vector() {constant=<rgba 0 0.6835 1 1>}, B=Constant3Vector() {constant=<rgba 1 0.1419 0.9555 1>}, Alpha=@n21.R)
   Alpha=Ceil(None=@n21.R)
  )
  n22 = WorldPosition()
  n23 = ParticlePositionWS()
  NamedRerouteDeclaration(None=ComponentMask(None=Subtract(A=@n22.XYZ, B=@n23)) {b=true}) {name=LinearGradient, nodeColor=<rgba 0.4062 0 1 1>}
  n24 = DynamicParameter() {parameterIndex=1, defaultValue=<rgba 5 1 1 1>, paramNames [DirectionX, DirectionY, DirectionZ, BendIntensity]}
  n25 = Multiply(A=@n24.RGB) {constB=-1}
  n26 = VertexColor()
  NamedRerouteDeclaration(None=Multiply(A=Multiply(A=@n25) {constB=0.04}, B=Power(Base=@n26.G) {constExponent=3})) {name=MovementOffset, nodeColor=<rgba 0 1 0.4609 1>}
  n27 = DynamicParameter() {defaultValue=<rgba 10 80 1 1>, paramNames [WobbleAmplitude, WobblePeriod, WobbleSpeed, WobbleOffsetRandom]}
  n28 = Time()
  n29 = ParticleRandom()
  n30 = Add(
   A=Divide(A=NamedRerouteUsage(), B=@n27.WobblePeriod) {constB=150}
   B=Multiply(A=Add(A=@n28, B=@n29), B=Multiply(A=@n27.WobbleSpeed) {constB=-1})
  )
  n31 = Multiply(A=Constant3Vector() {constant=<rgba 0 1 0 0>}, B=Cosine(None=@n30))
  n32 = Multiply(A=Sine(None=@n30), B=Constant3Vector() {constant=<rgba 1 0 0 0>})
  NamedRerouteDeclaration(None=Add(A=Multiply(A=@n31, B=@n27.WobbleAmplitude) {constB=5}, B=Multiply(A=@n32, B=@n27.WobbleAmplitude) {constB=5})) {name=WhipMain Wobble, nodeColor=<rgba 0.7812 1 0 1>}
  n33 = DynamicParameter() {parameterIndex=2, defaultValue=<rgba 45 100 1 1>, paramNames [DetailAmplitude, DetailPeriod, DetailSpeed, WobbleOffsetRandom]}
  n34 = Time()
  n35 = ParticleRandom()
  n36 = Add(A=Divide(A=NamedRerouteUsage(), B=@n33.DetailPeriod), B=Multiply(A=Add(A=@n34, B=@n35), B=Multiply(A=@n33.DetailSpeed) {constB=-1}))
  n37 = Multiply(A=Constant3Vector() {constant=<rgba 1 0 0 0>}, B=Sine(None=@n36))
  n38 = Multiply(A=Constant3Vector() {constant=<rgba 0 1 0 0>}, B=Cosine(None=@n36))
  NamedRerouteDeclaration(None=Add(A=Multiply(A=@n37, B=@n33.DetailAmplitude), B=Multiply(A=@n38, B=@n33.DetailAmplitude))) {name=DetailedWobble, nodeColor=<rgba 0 0.8359 1 1>}
 ]
}
