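# Niagara skydiving contrail cosmetic with a dessert theme: whipped cream, cherries, sprinkles and drops trailing from the character.
# Two spawn groups index skeletal mesh bones, likely both hands, anchoring can and cream meshes, wobbling ribbons, GPU sprite sprinkles and drops, and falling cherry meshes.
# ParaGlideLeanAlpha scales spawn rates, waveform wobble and the cream's dynamic material parameters, which also receive owner velocity.
# Time-of-day selectors tint colors, and front-end modifiers adjust sizes and speeds.
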NiagaraSystem /BRCosmetics/Athena/Effects/Particles/Skydiving/NS_Contrail_LilSplit/NS_Contrail_LilSplit
dependencies {
 usedDataInterfaces [
  &NiagaraDataInterfaceArrayColor
  &NiagaraDataInterfaceCurve
  &NiagaraDataInterfaceColorCurve
  &NiagaraDataInterfaceVectorCurve
 ]
 usedDynamicInputs [&SimulationPosition]
 usedModules [&SystemState]
}
emitters {
 Cherries {
  emitterUpdateScript [
   &EmitterState()
   &SpawnRate(
    Spawn Probability=0.5
    SpawnRate=&Lerp_Float(A=4, B=6, Alpha=$User.ParaGlideLeanAlpha)
    Use Spawn Probability=true
   )
  ]
  particleSpawnScript [
   &FN_SetSpawnGroup(Number of Spawn Groups=2)
   &InitializeParticle(
    Lifetime Max=1
    Lifetime Min=0.6
    Lifetime Mode=::NewEnumerator1
    Mass Mode=::NewEnumerator1
    Mesh Scale Mode=::NewEnumerator1
    Mesh Uniform Scale Max=3
    Mesh Uniform Scale Min=1.25
    UsePositionOffset=false
    Write Material Random=true
   ) {module=InitializeParticle001}
   &AddVelocity(Velocity=&RandomRangeVector(Minimum=<-150 -150 200>, Maximum=<150 150 400>))
   &SkeletalMeshLocation(
    Bone / Socket Index=$Particles.SpawnGroup
    Bone Sampling Mode=::NewEnumerator4
    MaskSpawnGroup=false
    Sampled Position Offset Coordinate Space:ENiagaraSkelMeshTransforms=::NewEnumerator4
    Skeletal Mesh=$User.SkeletalMesh
   )
   &InitialMeshOrientation(
    Mesh Orientation Mode=::NewEnumerator2
    Rotation=&RandomRangeVector(Minimum=<0 0 0>, Maximum=<1 0 1>)
    Use Rotation Vector=true
   )
   &ShapeLocation(MaskSpawnGroup=false, Override Seed=false, Sphere Radius=20, UseOwnerScale=true)
  ]
  particleUpdateScript [
   &ParticleState()
   &ScaleMeshSize(Scale Factor=&VectorFromFloat(Value=&FloatFromCurve(FloatCurve=DataInterface() {propertyValues {Curve {keys [<at 0 0>,<at 0.1 0.6>,<at 0.3 0.75>,<at 0.7 0.55>,<at 1 0>]}}})))
   &GravityForce(Gravity=<0 0 -250>)
   &Drag(Drag=2, Use Linear Drag=true, Use Rotational Drag=true)
   &SolveForcesAndVelocity(Clamp Velocity=false, Limit Acceleration=false)
  ]
  rendererClasses [&NiagaraMeshRendererProperties]
  renderers [{rendererClass=&NiagaraMeshRendererProperties, rendererIndex=0}]
 }
 Drops {
  simTarget=GPUComputeSim
  emitterUpdateScript [
   &EmitterState(Inactive Response=::NewEnumerator1, Life Cycle Mode=::NewEnumerator1, Loop Behavior=::NewEnumerator1, Loop Duration=3600, UseLoopDelay=false)
   &SpawnRate(
    SpawnGroup:NiagaraInt32=&RandomRangeInt(Minimum=0, Maximum=1)
    SpawnRate=60
    Use Spawn Probability=false
   )
  ]
  particleSpawnScript [
   &FN_SetSpawnGroup(Number of Spawn Groups=2)
   &InitializeParticle(
    Color=&SelectLinearColorFromArray(
     Array Sampling Mode=::NewEnumerator0
     Direct Array Index=?
     Color Selection Array=DataInterface() {propertyValues {ColorData [<rgba 1 0.8768 0.7948 1>, <rgba 0.2344 0.0759 0.0313 1>]}}
     Random Seed=0
    )
    Lifetime Max=0.5
    Lifetime Min=0.5
    Lifetime Mode=::NewEnumerator1
    UsePositionOffset=false
    Write Material Random=true
   ) {module=InitializeParticle001}
   &SkeletalMeshLocation(
    Bone / Socket Index=$Particles.SpawnGroup
    Bone Sampling Mode=::NewEnumerator4
    Inherit Velocity=false
    Interpolate Mesh Sampling=false
    MaskSpawnGroup=false
    Sampled Position Offset=<0 0 22>
    Skeletal Mesh=$User.SkeletalMesh
   )
   &Color(Color=&MultiplyLinearColor(
     Linear Color=$Particles.Initial.Color
     Scale Factor=&MakeVector4_FromLinearColor(LinearColor=&FN_Select_Color_from_Time_Of_Day(
       Morning=<rgba 0.7396 0.3609 0.285 1>
       Day=<rgba 1 1 1 1>
       Evening=<rgba 0.8177 0.8177 0.8177 1>
       Night=<rgba 0.4532 0.6063 0.9115 1>
       Morning Threshold=5
      ))
    ))
   /SetVariables(Particles.Velocity=&UniformRangedVector(Minimum=<-120 -120 100>, Maximum=<120 120 250>)) {isSetParameters=true}
   /SetVariables(Particles.SpriteSize=&Multiply_Vector2(
     A=<5 12>
     B=&Vector2DFromFloat(Value=&UniformRangedFloat(Minimum=0.45, Maximum=1.2))
    )) {isSetParameters=true}
  ]
  particleUpdateScript [
   &ParticleState()
   &GravityForce(Gravity=<0 0 -100>)
   &Drag(Drag=2, Use Linear Drag=true, Use Rotational Drag=false)
   &SolveForcesAndVelocity(Clamp Velocity=false, Limit Acceleration=false)
   &ScaleColor(
    Scale Alpha=&FloatFromCurve(FloatCurve=DataInterface() {propertyValues {Curve {keys [
         <at 0 0 RCIM_Cubic>
         {arriveTangent=1.6041, leaveTangent=1.6041, time=0.05, value=1.0113}
         {interpMode=RCIM_Cubic, leaveTangent=-1.0633, time=0.75, value=1}
         <at 0.99 0 RCIM_Cubic>
        ]}}})
    Scale RGB=<0.6 0.8 1>
    ScaleA=true
    ScaleRGB=false
   )
   &ScaleSpriteSize(Scale Factor=&Vector2DFromFloat(Value=&FloatFromCurve(
      FloatCurve=DataInterface() {propertyValues {Curve {keys [
          {arriveTangent=-0.4919, interpMode=RCIM_Cubic, leaveTangent=-0.4919, tangentMode=RCTM_User, time=0.25, value=2}
          {arriveTangent=-3.3301, interpMode=RCIM_Cubic, leaveTangent=-3.3302, tangentMode=RCTM_User, time=1, value=0}
         ]}}}
      Scale Curve=0.65
     )))
   &CameraOffset(Camera Offset Amount=8)
  ]
  rendererClasses [&NiagaraSpriteRendererProperties]
  renderers [{rendererClass=&NiagaraSpriteRendererProperties}]
 }
 HandSprinkles {
  simTarget=GPUComputeSim
  emitterUpdateScript [
   &EmitterState()
   /Game/Effects/Niagara/Modules/Spawn/SpawnRateFromMultipleSpawnGroups_2(Global Spawnrate=35)
  ]
  particleSpawnScript [
   &FN_SetSpawnGroup(Number of Spawn Groups=2)
   &InitializeParticle(
    Color=&SelectLinearColorFromArray(
     Array Sampling Mode=::NewEnumerator0
     Direct Array Index=?
     Color Selection Array=DataInterface() {propertyValues {ColorData [<rgba 0.8958 0.3028 0 1>, <rgba 1 0.0774 0.4681 1>, <rgba 0.1046 0.5635 1 1>, <rgba 0.47 0.1395 1 0.995>]}}
     Random Seed=0
    )
    Lifetime Max=0.7
    Lifetime Min=0.4
    Lifetime Mode=::NewEnumerator1
    Sprite Rotation Mode=::NewEnumerator2
    Sprite Size Max=<2.5 8>
    Sprite Size Min=<1 4>
    Sprite Size Mode=::NewEnumerator2
    UsePositionOffset=false
    Write Material Random=true
   )
   &Color(Color=&MultiplyLinearColor(
     Linear Color=$Particles.Initial.Color
     Scale Factor=&MakeVector4_FromLinearColor(LinearColor=&FN_Select_Color_from_Time_Of_Day(Morning=<rgba 1 1 1 1>, Day=<rgba 1 1 1 1>, Evening=<rgba 1 0.7676 0.7377 1>, Night=<rgba 0.7313 0.6437 1 1>))
    ))
   &ShapeLocation(MaskSpawnGroup=false, Override Seed=false, Sphere Radius=10, UseOwnerScale=true) {enabled=false}
   /SetVariables(
    Particles.RandomRadius:NiagaraFloat=&RandomRangeFloat(Minimum=10, Maximum=15)
    Particles.RandomRotationRate:NiagaraFloat=&RandomRangeFloat(Minimum=0.5, Maximum=1.5)
    Particles.RandomXAxis:Vector3f=&RandomRangeVector(Minimum=<-1 -1 -1>, Maximum=<1 1 1>)
    Particles.RandomYAxis:Vector3f=&RandomRangeVector(Minimum=<-1 -1 -1>, Maximum=<1 1 1>)
   ) {isSetParameters=true}
  ]
  particleUpdateScript [
   &ParticleState()
   &SpriteRotationRate(Rotation Rate=&RandomRangeFloat(Minimum=-150, Maximum=150))
   &SkeletalMeshLocation(Bone / Socket Index=$Particles.SpawnGroup, Bone Sampling Mode=::NewEnumerator4, MaskSpawnGroup=false, Skeletal Mesh=$User.SkeletalMesh)
   &ScaleSpriteSize(Uniform Curve Sprite Scale=DataInterface() {propertyValues {Curve {keys [
        {arriveTangent=6.8788, interpMode=RCIM_Cubic, leaveTangent=6.8788, tangentMode=RCTM_User, time=0, value=0.35}
        {arriveTangent=0.0013, interpMode=RCIM_Cubic, leaveTangent=0.0013, tangentMode=RCTM_User, time=0.1408, value=0.9847}
        {arriveTangent=-1.7881, interpMode=RCIM_Cubic, leaveTangent=-1.7881, tangentMode=RCTM_User, time=1, value=0.35}
       ]}}}) {enabled=false}
   &RotateAroundPoint(
    Derive the Rotation Axis=false
    Radius:NiagaraFloat=$Particles.RandomRadius
    Rotation Axis:Vector3f=$Particles.RandomYAxis
    Rotation Center:NiagaraPosition=&AddVectorToPosition(Position=$Particles.SkeletalMeshLocation.SampledPosition, Vector=<0 0 4>)
    Rotation Phase=1
    Rotation Rate:NiagaraFloat=$Particles.RandomRotationRate
    X Axis:Vector3f=$Particles.RandomXAxis
   )
  ]
  rendererClasses [&NiagaraSpriteRendererProperties]
  renderers [{rendererClass=&NiagaraSpriteRendererProperties}]
 }
 SplashRibbon {
  emitterUpdateScript [
   &EmitterState()
   &SpawnRate(SpawnRate=0, Use Spawn Probability=false)
   /Game/Effects/Niagara/Modules/Ribbons/FN_SpawnRibbon_v3(
    Number of Ribbons=2
    Skeleton Mesh=DataInterface() {propertyValues {
      DefaultMesh=null
      ExcludeBoneName=root
      OptionalComponentClass=null
      PreviewMesh=&M_Med_Soldier_04_Preview
      SkinningMode=SkinOnTheFly
      SoftSourceActor=null
      SourceMode=Default
      UvSetIndex=0
      WholeMeshLOD=-1
      bExcludeBone=false
      bReadDeformedGeometry=true
      bRequireCurrentFrameData=true
      ComponentTags []
      FilteredBones [hand_l,hand_r]
      FilteredSockets []
      MeshUserParameter {parameter {name=None, typeDefHandle {registeredTypeIndex=1}, varData []}}
      SamplingRegions []
     }}
   )
  ]
  particleSpawnScript [
   &FN_SetSpawnGroup(Number of Spawn Groups=2)
   /Game/Effects/Niagara/Modules/Ribbons/FN_InitializeRibbon_v5(
    Custom Offset=true
    Enable Skeletal Position=true
    Lifetime=1.1
    Local Vector Offset:Vector3f=<0 -6 20>
    Offset Amount=1
    Ribbon Width:NiagaraFloat=&FN_FrontEndFloatModifier(Gameplay=90, Front End=60)
    SkeletalMesh:NiagaraDataInterfaceSkeletalMesh=$User.SkeletalMesh
   )
   &AddVelocity(Velocity=&FN_FrontEndVector3Modifier(Gameplay=<0 0 300>, Front End=<0 0 200>)) {module=AddVelocity001}
   /SetVariables(Particles.Color=&LerpLinearColors(
     StartColor=&ColorFromCurve(
      ColorCurve=DataInterface() {propertyValues {
        AlphaCurve {keys [<at 0 1 RCIM_Cubic>]}
        BlueCurve {keys [<at 0 0.5325 RCIM_Cubic>,<at 0.1208 0.8594>,<at 0.5021 0.8229>]}
        GreenCurve {keys [<at 0 0.6706 RCIM_Cubic>,<at 0.1208 0.9282>,<at 0.5021 0.8897>]}
        RedCurve {keys [<at 0 0.7865 RCIM_Cubic>,<at 0.1208 1>,<at 0.5021 1>]}
       }}
      CurveIndex=&RandomRangeFloat(Minimum=0, Maximum=1)
     )
     EndColor=&ColorFromCurve(
      ColorCurve=DataInterface() {propertyValues {
        AlphaCurve {keys [<at 0 1 RCIM_Cubic>]}
        BlueCurve {keys [<at 0 0.5325 RCIM_Cubic>,<at 0.1208 0.8594>,<at 0.6586 0.8385>]}
        GreenCurve {keys [<at 0 0.6706 RCIM_Cubic>,<at 0.1208 0.9282>,<at 0.6586 0.9084>]}
        RedCurve {keys [<at 0 0.7865 RCIM_Cubic>,<at 0.1208 1>,<at 0.6586 1>]}
       }}
      CurveIndex=&RandomRangeFloat(Minimum=0, Maximum=1)
     )
     LerpFactor=&MakeFloatFromInt(Int=$Particles.SpawnGroup)
    )) {isSetParameters=true}
   &Color(Color=&MultiplyLinearColor(
     Linear Color=$Particles.Color
     Scale Factor=&MakeVector4_FromLinearColor(LinearColor=&FN_Select_Color_from_Time_Of_Day(
       Morning=<rgba 1 0.7532 0.7135 1>
       Day=<rgba 1 1 1 1>
       Evening=<rgba 0.9323 0.6312 0.6312 1>
       Night=<rgba 0.3278 0.4247 0.6771 1>
       Morning Threshold=5
      ))
    ))
   /SetVariables(Particles.MaterialRandom:NiagaraFloat=&RandomRangeFloat(Minimum=0, Maximum=1)) {isSetParameters=true}
  ]
  particleUpdateScript [
   &ScaleVelocity(Velocity Scale=&MakeVector(
     X=0
     Y=0
     Z=&Multiply_Float(
      A=0.9
      B=&FloatFromCurve(
       FloatCurve=DataInterface() {propertyValues {Curve {keys [<at 0 0.9>,<at 1 3>]}}}
       CurveIndex=&Divide_Float(A=&VectorLength(Vector=$Engine.Owner.Velocity), B=9000)
      )
     )
    )) {enabled=false, module=ScaleVelocity001}
   &InheritSourceMovement(
    Applied Position Falloff Distance=100
    Applied Position Scale=<-0.08 -0.08 0>
    ApplyForce=false
    ApplyPosition=true
    ApplyPositionFalloff=false
   ) {enabled=false}
   &InheritVelocity(Inherited Velocity Amount Scale=<-0.008 -0.008 0>, Inherited Velocity Speed Limit=50, Speed Limit=true, UseSpeedThreshold=false) {enabled=false}
   &SolveForcesAndVelocity(Clamp Velocity=false, Limit Acceleration=false)
   /Game/Effects/Niagara/Modules/Ribbons/FN_UpdateRibbon_v6(Enable Skeletal Position=true, SkeletalMesh:NiagaraDataInterfaceSkeletalMesh=$User.SkeletalMesh)
   &DynamicMaterialParameters(
    Index 0 Param 1=&MaskFloatBySpawnGroup(Masked Float=0.5, Passthrough Float=1)
    Index 0 Param 2=&FloatFromCurve(
     FloatCurve=DataInterface() {propertyValues {Curve {keys [
         {arriveTangent=-3.1795, interpMode=RCIM_Cubic, leaveTangent=-3.1795, tangentMode=RCTM_User, time=0, value=1}
         {arriveTangent=-0.0377, interpMode=RCIM_Cubic, leaveTangent=0.0167, tangentMode=RCTM_Break, time=1, value=0}
        ]}}}
     Scale Curve=0.5
    )
    Index 0 Param 3=0.2
    Index 0 Param 4=0.2
    Module.Param0WriteEnabled=true
    Module.Param1WriteEnabled=true
    Module.Param2WriteEnabled=true
    Module.Param3WriteEnabled=true
   )
   &ScaleRibbonWidth(Ribbon Width Scale=&FloatFromCurve(FloatCurve=DataInterface() {propertyValues {Curve {keys [
         {arriveTangent=4.1785, interpMode=RCIM_Cubic, leaveTangent=4.1785, tangentMode=RCTM_User, time=0, value=0}
         {arriveTangent=1.1179, interpMode=RCIM_Cubic, leaveTangent=1.1179, tangentMode=RCTM_User, time=0.2, value=0.85}
         {arriveTangent=0.0224, interpMode=RCIM_Cubic, leaveTangent=0.0224, tangentMode=RCTM_User, time=0.5, value=1}
        ]}}}))
   /SetVariables(Particles.Position=&AddVectorToPosition(
     Position=$Particles.Position
     Vector=&Multiply_VectorByFloat(
      Vector=<0 1 0>
      Float=&Waveform(
       Global Input Position=&ReturnNormalizedExecIndex()
       Global Amplitude Scale=&ReturnNormalizedExecIndex()
       "Waveform [1]"=::NewEnumerator1
       "[1] Frequency"=2
       "[1] Amplitude Min/Max"=<-1.2 1.2>
       "[1] Scale"=&Lerp_Float(A=0.6, B=1, Alpha=$User.ParaGlideLeanAlpha)
      )
     )
    )) {isSetParameters=true}
   /SetVariables(Particles.Position=&AddVectorToPosition(
     Position=$Particles.Position
     Vector=&Multiply_VectorByFloat(
      Vector=<1 0 0>
      Float=&Waveform(
       Global Input Position=&ReturnNormalizedExecIndex()
       Global Amplitude Scale=&ReturnNormalizedExecIndex()
       "Waveform [1]"=::NewEnumerator0
       "[1] Frequency"=2
       "[1] Amplitude Min/Max"=<-1.2 1.2>
       "[1] Scale"=&Lerp_Float(A=0.6, B=1, Alpha=$User.ParaGlideLeanAlpha)
      )
     )
    )) {isSetParameters=true}
   &ScaleColor(
    Scale Alpha=&FloatFromCurve(
     FloatCurve=DataInterface() {propertyValues {Curve {keys [<at 0 0>,<at 0.1 1>,<at 1 1>]}}}
     Scale Curve=&FN_Select_Float_from_Time_Of_Day(Morning=2, Day=1, Evening=1, Night=1)
    )
    Scale RGB=<2 2 2>
    ScaleA=true
    ScaleRGB=true
   ) {module=ScaleColor001}
  ]
  rendererClasses [&NiagaraRibbonRendererProperties]
  renderers [{rendererClass=&NiagaraRibbonRendererProperties}]
 }
 Sprinkles {
  simTarget=GPUComputeSim
  emitterUpdateScript [
   &EmitterState()
   &SpawnRateFromMultipleSpawnGroups_2(Global Spawnrate:NiagaraFloat=&Lerp_Float(A=20, B=25, Alpha=$User.ParaGlideLeanAlpha))
  ]
  particleSpawnScript [
   &FN_SetSpawnGroup(Number of Spawn Groups=2)
   &InitializeParticle(
    Color=&SelectLinearColorFromArray(
     Array Sampling Mode=::NewEnumerator0
     Direct Array Index=?
     Color Selection Array=DataInterface() {propertyValues {ColorData [<rgba 0.8958 0.3028 0 1>, <rgba 1 0.0774 0.4681 1>, <rgba 0.1046 0.5635 1 1>, <rgba 0.47 0.1395 1 0.995>]}}
     Random Seed=0
    )
    Lifetime Max=1
    Lifetime Min=0.5
    Lifetime Mode=::NewEnumerator1
    Sprite Rotation Mode=::NewEnumerator2
    Sprite Size Max=<4 12>
    Sprite Size Min=<3 7>
    Sprite Size Mode=::NewEnumerator2
    UsePositionOffset=true
    Write Material Random=true
   )
   &Color(Color=&MultiplyLinearColor(
     Linear Color=$Particles.Initial.Color
     Scale Factor=&MakeVector4_FromLinearColor(LinearColor=&FN_Select_Color_from_Time_Of_Day(Morning=<rgba 1 1 1 1>, Day=<rgba 1 1 1 1>, Evening=<rgba 1 0.7676 0.7377 1>, Night=<rgba 0.7313 0.6437 1 1>))
    ))
   &SkeletalMeshLocation(
    Bone / Socket Index=$Particles.SpawnGroup
    Bone Sampling Mode=::NewEnumerator4
    MaskSpawnGroup=false
    Sampled Position Offset Coordinate Space:ENiagaraSkelMeshTransforms=::NewEnumerator0
    Skeletal Mesh=$User.SkeletalMesh
   )
   &ShapeLocation(MaskSpawnGroup=false, Override Seed=false, Sphere Radius=18, UseOwnerScale=true)
   &AddVelocity(
    Velocity=&Multiply_Vector(
     A=&RandomRangeVector(Minimum=<0 0 150>, Maximum=<0 0 330>)
     B=&VectorFromFloat(Value=&Lerp_Float(A=1, B=1.25, Alpha=$User.ParaGlideLeanAlpha))
    )
    Velocity Speed Scale=&FN_FrontEndFloatModifier(Gameplay=1, Front End=0.6)
   )
  ]
  particleUpdateScript [
   &ParticleState()
   &ScaleSpriteSize(Uniform Curve Sprite Scale=DataInterface() {propertyValues {Curve {keys [
        {arriveTangent=9.0725, interpMode=RCIM_Cubic, leaveTangent=9.0725, tangentMode=RCTM_User, time=0, value=0}
        {arriveTangent=0.4662, interpMode=RCIM_Cubic, leaveTangent=0.4662, tangentMode=RCTM_User, time=0.1408, value=0.9847}
        {arriveTangent=-2.3238, interpMode=RCIM_Cubic, leaveTangent=-2.3238, tangentMode=RCTM_User, time=1, value=0.2}
       ]}}})
   &SolveForcesAndVelocity(Clamp Velocity=false, Limit Acceleration=false)
   /SetVariables(Particles.Position=&AddVectorToPosition(
     Position=$Particles.Position
     Vector=&Multiply_VectorByFloat(
      Vector=<0 1 0>
      Float=&Waveform(
       Global Input Position=&ReturnNormalizedExecIndex()
       Global Amplitude Scale=&ReturnNormalizedExecIndex()
       "Waveform [1]"=::NewEnumerator1
       "[1] Frequency"=&Lerp_Float(A=2, B=5, Alpha=$User.ParaGlideLeanAlpha)
       "[1] Amplitude Min/Max"=<-2 2>
       "[1] Scale"=&Lerp_Float(A=0.6, B=2.5, Alpha=$User.ParaGlideLeanAlpha)
      )
     )
    )) {isSetParameters=true}
   /SetVariables(Particles.Position=&AddVectorToPosition(
     Position=$Particles.Position
     Vector=&Multiply_VectorByFloat(
      Vector=<1 0 0>
      Float=&Waveform(
       Global Input Position=&ReturnNormalizedExecIndex()
       Global Amplitude Scale=&ReturnNormalizedExecIndex()
       "Waveform [1]"=::NewEnumerator0
       "[1] Frequency"=&Lerp_Float(A=2, B=4, Alpha=$User.ParaGlideLeanAlpha)
       "[1] Amplitude Min/Max"=<-2 2>
       "[1] Scale"=&Lerp_Float(A=1, B=2, Alpha=$User.ParaGlideLeanAlpha)
      )
     )
    )) {isSetParameters=true}
  ]
  rendererClasses [&NiagaraSpriteRendererProperties]
  renderers [{rendererClass=&NiagaraSpriteRendererProperties}]
 }
 WhipCan {
  emitterUpdateScript [
   &EmitterState(Life Cycle Mode=::NewEnumerator1, Loop Behavior=::NewEnumerator1, Loop Delay=0.045, Loop Duration=1, UseLoopDelay=false)
   &SpawnBurst_Instantaneous(Spawn Count=2, Use Spawn Probability=false, UseLoopCountLimit=false)
  ]
  particleSpawnScript [
   &FN_SetSpawnGroup(Number of Spawn Groups=2)
   &InitializeParticle(
    Color Mode=::NewEnumerator0
    Lifetime Max=1
    Lifetime Min=0.5
    Lifetime Mode=::NewEnumerator1
    Mesh Scale=<1.2 1.2 1>
    Mesh Scale Mode=::NewEnumerator4
    Sprite Rotation Mode=::NewEnumerator2
    Sprite Size Max=<3.5 12>
    Sprite Size Min=<2 8>
    Sprite Size Mode=::NewEnumerator2
    UsePositionOffset=false
    Write Material Random=true
   )
   &InitialMeshOrientation(Mesh Orientation Mode=::NewEnumerator2, Rotation=<0 0 0.5>, Rotation Coordinate Space=::NewEnumerator0, Use Rotation Vector=true)
  ]
  particleUpdateScript [
   &ParticleState(Kill Particles When Lifetime Has Elapsed=false)
   &SkeletalMeshLocation(
    Bone / Socket Index=$Particles.SpawnGroup
    Bone Sampling Mode=::NewEnumerator4
    MaskSpawnGroup=false
    Orientation Sampling=::NewEnumerator1
    Sampled Position Offset=<0 0 12>
    Sampled Position Offset Coordinate Space:ENiagaraSkelMeshTransforms=::NewEnumerator4
    Skeletal Mesh=$User.SkeletalMesh
   )
  ]
  rendererClasses [&NiagaraMeshRendererProperties]
  renderers [
   {rendererClass=&NiagaraMeshRendererProperties, rendererIndex=0}
   {rendererClass=&NiagaraMeshRendererProperties, rendererIndex=1}
  ]
 }
 WhipMesh {
  emitterUpdateScript [
   &EmitterState(Life Cycle Mode=::NewEnumerator1, Loop Behavior=::NewEnumerator1, Loop Duration=0, UseLoopDelay=false)
   &SpawnBurst_Instantaneous(Spawn Count=2, Spawn Probability=HlslExpression() {hlslExpression="Emitter.LoopCount == 0"}, Use Spawn Probability=true)
  ]
  particleSpawnScript [
   &FN_SetSpawnGroup(Number of Spawn Groups=2)
   &InitializeParticle(
    Lifetime=0.8
    Mass Mode=::NewEnumerator1
    Mesh Scale Mode=::NewEnumerator3
    Mesh Uniform Scale=&FN_FrontEndFloatModifier(Gameplay=1, Front End=0.75)
    UsePositionOffset=false
    Write Material Random=true
   ) {module=InitializeParticle001}
   &InitialMeshOrientation(
    Mesh Orientation Mode=::NewEnumerator2
    Rotation=&RandomRangeVector(Minimum=<0 0 -1>, Maximum=<0 0 1>)
    Rotation Coordinate Space=::NewEnumerator0
    Use Rotation Vector=true
   )
  ]
  particleUpdateScript [
   &ParticleState(Kill Particles When Lifetime Has Elapsed=false)
   &ScaleMeshSize(Scale Factor=&VectorFromCurve(VectorCurve=DataInterface() {propertyValues {XCurve {keys [<at 0.1 0.125>,<at 0.7 1>]}, YCurve {keys [<at 0.1 0.125>,<at 0.7 1>]}, ZCurve {keys [<at 0 0>,<at 1 1>]}}}))
   &SkeletalMeshLocation(
    Bone / Socket Index=$Particles.SpawnGroup
    Bone Sampling Mode=::NewEnumerator4
    MaskSpawnGroup=false
    Orientation Sampling=::NewEnumerator1
    Sampled Position Offset=<0 0 22>
    Sampled Position Offset Coordinate Space:ENiagaraSkelMeshTransforms=::NewEnumerator4
    Skeletal Mesh=$User.SkeletalMesh
   ) {module=SkeletalMeshLocation001}
   &UpdateMeshOrientation(Rotation Vector=&RandomRangeVector(Minimum=<0 0 -1>, Maximum=<0 0 1>))
   &DynamicMaterialParameters(
    Index 0 Param 1=&FloatFromCurve(
     FloatCurve=DataInterface() {propertyValues {Curve {keys [{interpMode=RCIM_Cubic, leaveTangent=8.0137, tangentMode=RCTM_Break, time=0.5, value=0}, <at 0.8 1>]}}}
     Scale Curve=&Lerp_Float(A=12, B=15, Alpha=$User.ParaGlideLeanAlpha)
    )
    Index 0 Param 2=&FloatFromCurve(
     FloatCurve=DataInterface() {propertyValues {Curve {keys [{interpMode=RCIM_Cubic, leaveTangent=3.7914, tangentMode=RCTM_Break, time=0, value=0}, <at 0.4 1>]}}}
     Scale Curve=&Lerp_Float(A=85, B=60, Alpha=$User.ParaGlideLeanAlpha)
    )
    Index 0 Param 3=3
    Index 0 Param 4=5
    Index 1 Param 1=HlslExpression() {hlslExpression=Engine.Owner.Velocity.x}
    Index 1 Param 2=HlslExpression() {hlslExpression=Engine.Owner.Velocity.y}
    Index 1 Param 4=50
    Index 2 Param 1=&FloatFromCurve(
     FloatCurve=DataInterface() {propertyValues {Curve {keys [<at 0 0>,<at 0.45 1>]}}}
     Scale Curve=&FN_FrontEndFloatModifier(Gameplay=40, Front End=25)
    )
    Index 2 Param 2=40
    Index 2 Param 3=5
    Index 2 Param 4=&FN_FrontEndFloatModifier(
     Gameplay=&FloatFromCurve(FloatCurve=DataInterface() {propertyValues {Curve {keys [<at 0 0>,<at 0.55 0>,<at 1 1>]}}}, Scale Curve=65)
     Front End=&FloatFromCurve(FloatCurve=DataInterface() {propertyValues {Curve {keys [<at 0 0>,<at 0.55 0>,<at 1 1>]}}}, Scale Curve=40)
    )
    Module.Index0Param0Write=true
    Module.Index0Param1Write=true
    Module.Index0Param2Write=true
    Module.Index0Param3Write=false
    Module.Index1Param0Write=true
    Module.Index1Param1Write=true
    Module.Index1Param2Write=true
    Module.Index1Param3Write=false
    Module.Index2Param0Write=true
    Module.Index2Param1Write=true
    Module.Index2Param2Write=true
    Module.Index2Param3Write=true
    Write Parameter Index 1=true
    Write Parameter Index 2=true
   )
  ]
  rendererClasses [&NiagaraMeshRendererProperties]
  renderers [{rendererClass=&NiagaraMeshRendererProperties, rendererIndex=0}]
 }
}
summary {
 systemName=NS_Contrail_LilSplit
 emitters [
  {emitterName=HandSprinkles, simTarget=GPUComputeSim, rendererClasses [&NiagaraSpriteRendererProperties]}
  {emitterName=WhipCan, rendererClasses [&NiagaraMeshRendererProperties]}
  {emitterName=Sprinkles, simTarget=GPUComputeSim, rendererClasses [&NiagaraSpriteRendererProperties]}
  {emitterName=Cherries, rendererClasses [&NiagaraMeshRendererProperties]}
  {emitterName=Drops, simTarget=GPUComputeSim, rendererClasses [&NiagaraSpriteRendererProperties]}
  {emitterName=SplashRibbon, rendererClasses [&NiagaraRibbonRendererProperties]}
  {emitterName=WhipMesh, rendererClasses [&NiagaraMeshRendererProperties]}
 ]
 userVariables [
  User.bIsFrontEnd {type=NiagaraBool}
  User.bIsFrontEndPreview {type=NiagaraBool}
  User.ParaGlideLeanAlpha {type=NiagaraFloat, defaultValue=0, description=""}
  User.RibbonIDOffset {type=NiagaraInt32}
  User.SkeletalMesh {
   type=NiagaraDataInterfaceSkeletalMesh
   defaultValue {
    struct=NiagaraExt_VariableValue_DataInterface
    value {dataInterface=&self:NiagaraDataInterfaceSkeletalMesh_0}
   }
  }
 ]
}
